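# Physics asset for the InkHoop bass instrument cosmetic, defining a short hanging chain with two tags attached to the handle.
# Four capsules approximate the handle, while each chain link and tag has one small capsule.
# Constraints link the chain root to the handle, run through three links, and branch to both tags from the last.
# Locked axes near the root and on the tags restrict movement mostly to swing2, with the lower links moving more freely.
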PhysicsAsset /SparksCosmetics/Instrument/Bass/InkHoop/Meshes/InkHoop_Bass_Physics
bodies {
 chain_01 [{shapeType=Capsule, length=2.3378, radius=1.25, shapeName=chain_01_capsule, center=<1.5 0.5 0>, rotation=<rot 90 0 0>}]
 chain_02 [{shapeType=Capsule, length=1.8478, radius=1.4, shapeName=chain_02_capsule, center=<0.5 0 0>, rotation=<rot 90 0 0>}]
 chain_03 [{shapeType=Capsule, length=0.4878, radius=1.5306, shapeName=chain_03_capsule, center=<0.95 0 0>, rotation=<rot 90 0 0>}]
 chain_root [{shapeType=Capsule, length=1.3428, radius=1.1672, shapeName=chain_root_capsule, center=<1 0 0>, rotation=<rot 90 0 0>}]
 handle [
  {shapeType=Capsule, length=31.3272, radius=9.3638, shapeName=handle_capsule, center=<-6.7256 -0.2582 -6.3863>}
  {shapeType=Capsule, length=25.5031, radius=8.9331, shapeName=handle_capsule, center=<-7.586 -8.9303 -5.3047>, rotation=<rot 0.0003 0 15.1012>}
  {shapeType=Capsule, length=8.6553, radius=1.8762, shapeName=handle_capsule, center=<2.4813 -10.9189 -15.2718>, rotation=<rot 0.0001 0 2.9715>}
  {shapeType=Capsule, length=37.538, radius=3.153, shapeName=handle_capsule, center=<-0.2499 -0.1068 31.6893>}
 ]
 tag_01 [{shapeType=Capsule, length=4, radius=2, shapeName=tag_01_capsule, center=<3.2 1.3 -0.5>, rotation=<rot -0.0004 -87.0848 -76.6001>}]
 tag_02 [{shapeType=Capsule, length=4, radius=1.75, shapeName=tag_02_capsule, center=<2.8 1.2 1.8>, rotation=<rot 0.0001 92.1094 -56.0533>}]
}
constraints [
 {bone1Name=chain_03, bone2Name=chain_02, swing1LimitDegrees=45, swing2LimitDegrees=90, twistLimitDegrees=10}
 {bone1Name=chain_02, bone2Name=chain_01, swing1LimitDegrees=30, swing2LimitDegrees=90, twistLimitDegrees=10}
 {bone1Name=chain_01, bone2Name=chain_root, swing1LimitDegrees=0, swing1Motion=Locked, swing2LimitDegrees=90, twistLimitDegrees=2}
 {
  bone1Name=chain_root
  bone2Name=handle
  swing1LimitDegrees=0
  swing1Motion=Locked
  swing2LimitDegrees=180
  twistLimitDegrees=5
  twistMotion=Locked
 }
 {
  bone1Name=tag_01
  bone2Name=chain_03
  swing1LimitDegrees=45
  swing1Motion=Locked
  swing2LimitDegrees=45
  twistLimitDegrees=45
  twistMotion=Locked
 }
 {
  bone1Name=tag_02
  bone2Name=chain_03
  swing1LimitDegrees=45
  swing1Motion=Locked
  swing2LimitDegrees=55
  twistLimitDegrees=45
  twistMotion=Locked
 }
]
