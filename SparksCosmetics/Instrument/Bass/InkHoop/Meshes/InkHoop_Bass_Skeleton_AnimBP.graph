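# Animation Blueprint for the InkHoop Bass cosmetic instrument, adding secondary physics motion to its skeleton.
# A RigidBody node drives the output pose from the InkHoop bass physics asset, simulated relative to the handle bone with damped component motion.
# Each update sets its external force vector according to skydiving and parachute state.
# Anim Dynamics nodes for two tag bones and debug-draw calls are present but unconnected.
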AnimBlueprint /SparksCosmetics/Instrument/Bass/InkHoop/Meshes/InkHoop_Bass_Skeleton_AnimBP
target_skeleton=&/InkHoop_Bass_Skeleton
graphs [
 AnimGraph {nodes [
   n1 = Anim Dynamics - Bone: tag_01(Alpha=1, ExternalForce="0,0,0") {
    angularDampingOverride=0.99
    angularSpringConstant=6
    bAngularSpring=true
    bOverrideAngularDamping=true
    bUseGravityOverride=true
    alphaBoolBlend {blendOption=Linear}
    boundBone {boneName=tag_01}
    componentAppliedLinearAccClamp=<500 500 500>
    componentLinearAccScale=<0.5 0.5 0.5>
    componentLinearVelScale=<0.5 0.5 0.5>
    gravityOverride {z=-980}
    physicsBodyDefinitions [{
      boundBone {boneName=tag_01}
      boxExtents=<4 1 4>
      constraintSetup {angularLimitsMax=<0 55 10>, angularLimitsMin=<0 -45 0>, angularTarget=<0 0 0>}
      localJointOffset=<1 0.35 0>
     }]
    relativeSpaceBone {boneName=handle}
   }
   Anim Dynamics - Bone: tag_02(ComponentPose=@n1, Alpha=1, ExternalForce="0,0,0") {
    angularDampingOverride=0.99
    angularSpringConstant=6
    bAngularSpring=true
    bOverrideAngularDamping=true
    bUseGravityOverride=true
    linearDampingOverride=0.9
    alphaBoolBlend {blendOption=Linear}
    boundBone {boneName=tag_02}
    componentAppliedLinearAccClamp=<500 500 500>
    componentLinearAccScale=<0.5 0.5 0.5>
    componentLinearVelScale=<0.5 0.5 0.5>
    gravityOverride {z=-980}
    physicsBodyDefinitions [{
      boundBone {boneName=tag_02}
      boxExtents=<4 1 4>
      constraintSetup {angularLimitsMax=<0 55 0>, angularLimitsMin=<0 -45 -10>, angularTarget=<0 0 0>}
      localJointOffset=<0.5 -0.2921 0>
     }]
    relativeSpaceBone {boneName=handle}
   }
   n2 = RigidBody(ComponentPose=Local To Component(LocalPose=Input Pose()), ExternalForce=Get GravityOverride()) {
    bOverrideWorldGravity=true
    overridePhysicsAsset=&/InkHoop_Bass_Physics
    simulationSpace=BaseBoneSpace
    alphaBoolBlend {blendOption=Linear}
    baseBoneRef {boneName=handle}
    componentAppliedLinearAccClamp=<3000 3000 3000>
    componentLinearAccScale=<0.4 0.4 0.4>
    componentLinearVelScale=<0.2 0.2 0.2>
    overrideWorldGravity {z=-980}
    simSpaceSettings {
     maxAngularAcceleration=1000
     maxAngularVelocity=6
     maxLinearAcceleration=700
     maxLinearVelocity=600
     velocityScaleZ=0
     worldAlpha=0.6
     externalLinearDragV=<0.2 0.2 0.02>
    }
   }
   Output Pose(Result=Component To Local(ComponentPose=@n2))
  ]}
 EventGraph {nodes [
   Debug Player Velocities Function!(
    execute
    CharacterPartAnimInstance:Custom Character Part Anim Instance Object Reference=Self-Reference()
    SkelMeshComponent:Scene Component Object Reference
    DeltaTime:double=0
    DebugSpace:DebugDrawSpace Enum=NewEnumerator1
    BoneName:Name=root
    DebugDrawOffset:Vector="0,0,105"
    AngularDebugDrawOffset:Vector="0,0,0"
    DrawDebug:bool=true
    Debug Scalar:double=50
    Foreground:EDrawDebugSceneDepthPriorityGroup Enum=Foreground
    TextLocationOffset:Vector="20,20,20"
   ) -> (
    then
    Component VelocityX Vector:Vector
    Component VelocityY Vector:Vector
    Component VelocityZ Vector:Vector
    RotationRateDegrees:float
    BoneY Axis1:double
    BoneX Axis1:double
    BoneZ Axis1:double
    BoneLocationPassed1:Vector
    ReturnValue3:Vector
    ReturnValue4:Vector
    ReturnValue5:Vector
    ComponentRightVectorDot1:double
    ComponentFwdVectorDot1:double
    ComponentUpVectorDot1:double
    PawnVelocity1:Vector
   )
   > DrawJointDebugFunction!(
    execute
    SkelMeshComponent:Scene Component Object Reference=GetPartSkeletalMeshComponent()
    JointOffset:Vector="0, 0, 0"
    JointName:Array of Names=Make Array("[0]":Name=handle) -> (Array:Array of Names)
    Scale:double=10
    Duration:double=0
    Thickness:double=0.7
    DebugDraw:bool=true
    Foreground:EDrawDebugSceneDepthPriorityGroup Enum=Foreground
   ) -> (then)
   Event BlueprintUpdateAnimation()
   > n1 = Branch(Condition=Get bIsSkyDiving())
   > n2 = Branch(Condition=Get bIsParachuteOpened())
   > Set GravityOverride(GravityOverride="0,0,-400")
   Set GravityOverride(execute=@n2.else, GravityOverride="0,0,10000")
   Set GravityOverride(execute=@n1.else, GravityOverride="0,0,-980")
   TryGetPawnOwner()
  ]}
]
variables [GravityOverride {type=Vector}]
